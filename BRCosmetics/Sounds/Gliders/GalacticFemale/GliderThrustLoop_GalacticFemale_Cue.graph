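# Looping thrust sound cue for the GalacticFemale glider cosmetic, using glider attenuation and a cosmetic glider loop properties sheet.
# A Mixer blends two looping waves, a base thrust loop and a movement loop, each through its own Continuous Modulator.
# Both modulators read the AnyDirection parameter to scale pitch and volume, so the movement layer's volume swings over a wider relative range than the base loop as input rises.
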SoundCue /BRCosmetics/Sounds/Gliders/GalacticFemale/GliderThrustLoop_GalacticFemale_Cue
root=Mixer_0
nodes [
 {id=Mixer_0, children [ModulatorContinuous_2, ModulatorContinuous_3, null], properties {inputVolume [1,1,1]}}
 {
  id=ModulatorContinuous_2
  children [WavePlayer_0]
  properties {
   pitchModulationParams {default=0, maxOutput=1.2, minOutput=1, parameterName=AnyDirection}
   volumeModulationParams {default=0, maxOutput=0.8, minOutput=0.5, parameterName=AnyDirection}
  }
 }
 {
  id=ModulatorContinuous_3
  children [WavePlayer_1]
  properties {
   pitchModulationParams {default=0, minOutput=0.8, parameterName=AnyDirection}
   volumeModulationParams {default=0, maxInput=1.5, maxOutput=0.6, minOutput=0.1, parameterName=AnyDirection}
  }
 }
 {id=WavePlayer_0, properties {bLooping=true, soundWaveAssetPtr=&/Glider_GalacticFemale_Thrust_LP}}
 {
  id=WavePlayer_1
  properties {bLooping=true, soundWaveAssetPtr=&/Glider_GalacticFemale_Thrust_LP_Move}
 }
]
properties {
 attenuationSettings=&Fort_Player_Glider_Thrust_Athena_Att
 audioPropertiesSheet=&APS_Cosmetic_Glider_Loop
 maxDistance=12500
 soundClassObject=&Fort_Gliders_Land
 volumeMultiplier=0.9
 concurrencySet [null]
}
